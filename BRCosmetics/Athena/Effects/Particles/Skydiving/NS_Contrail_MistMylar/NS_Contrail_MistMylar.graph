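# Niagara skydiving contrail cosmetic (MistMylar) that attaches mesh trails, spirals, lightning, sprite smoke and swaying mask meshes to two bones or sockets on the skeletal mesh.
# Each emitter partitions its particles to choose an attachment and offset, re-sampled every update.
# ParaGlideLeanAlpha shifts trails and spirals toward warm gold and intensifies scale, speed, spawn rate or sway across most emitters.
# Front-end, time-of-day and platform modifiers adjust brightness, tint and opacity.
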NiagaraSystem /BRCosmetics/Athena/Effects/Particles/Skydiving/NS_Contrail_MistMylar/NS_Contrail_MistMylar
dependencies {
 usedDataInterfaces [
  &NiagaraDataInterfaceArrayInt32
  &NiagaraDataInterfaceArrayFloat3
  &NiagaraDataInterfaceArrayFloat
  &NiagaraDataInterfaceMeshRendererInfo
  /Script/Niagara.NiagaraDataInterfaceArrayQuat
  &NiagaraDataInterfaceCurve
  &NiagaraDataInterfacePlatformSet
  &NiagaraDataInterfaceSpriteRendererInfo
  &NiagaraDataInterfaceColorCurve
  &NiagaraDataInterfaceSkeletalMesh
 ]
 usedDynamicInputs []
 usedModules [
  &self:SystemSpawnScript.SystemScriptSource.SystemScriptGraph.NiagaraNodeAssignment_0.SetVariables_#1
  &SystemState
 ]
}
emitters {
 Lightning {
  emitterUpdateScript [
   &EmitterState(
    Inactive Response=::NewEnumerator1
    Life Cycle Mode=::NewEnumerator1
    Loop Delay=&RandomRangeFloat(Minimum=0, Maximum=0.3)
    Loop Duration=&RandomRangeFloat(Minimum=0.5, Maximum=1)
    Recalculate Duration Each Loop=true
    UseLoopDelay=true
   )
   &SpawnRate(SpawnRate=12, Use Spawn Probability=false)
  ]
  particleSpawnScript [
   &PartitionParticles()
   &InitializeParticle(
    Color=<rgba 1 1 1 3>
    Lifetime Max=0.5
    Lifetime Min=0.15
    Lifetime Mode=::NewEnumerator1
    Mesh Scale Max=<1.2 1.2 1.2>
    Mesh Scale Mode=::NewEnumerator2
    UsePositionOffset=false
    Write Material Random=true
   )
   &InitialMeshOrientation(
    Mesh Orientation Mode=::NewEnumerator2
    Rotation=&RandomRangeVector(Minimum=<0 0 0>, Maximum=<0 0 1>)
    Use Rotation Vector=true
   )
   &ShapeLocation(
    Box Size=<5 5 5>
    MaskSpawnGroup=false
    Offset=&SelectVectorFromArray(
     Direct Array Index=$Particles.PartitionParticles.Partition
     Vector Selection Array=DataInterface() {propertyValues {FloatData [<2 2 0>, <-3 -3 0>]}}
    )
    Offset Mode=::NewEnumerator0
    Override Seed=false
    Shape Primitive=::NewEnumerator2
    Surface Only=false
    UseOwnerScale=true
   )
   &AddVelocity(Velocity=<0 0 10>)
   /SetVariables(Particles.MeshIndex2:NiagaraInt32=&RandomRangeInt(Minimum=0, Maximum=1)) {isSetParameters=true}
  ]
  particleUpdateScript [
   &ParticleState()
   &SkeletalMeshLocation(
    Bone / Socket Index=&SelectIntFromArray(Direct Array Index=$Particles.PartitionParticles.Partition, Int Selection Array=DataInterface() {propertyValues {IntData [0,1]}})
    Bone Sampling Mode=::NewEnumerator4
    MaskSpawnGroup=false
    Orientation Sampling=::NewEnumerator1
    Position Sampling=::NewEnumerator1
    Sampled Position Offset=&AddVectorToPosition(Position=$Particles.Initial.ShapeLocation.ShapePosition, Vector=<0 0 0>)
    Skeletal Mesh=$System.Skeletal Mesh
   )
   &DynamicMaterialParameters(
    Index 0 Param 1=&FloatFromCurve(FloatCurve=DataInterface() {propertyValues {Curve {keys [<at 0 0>,<at 0.7 1>]}}})
    Index 0 Param 2=0.7
    Index 0 Param 3=&FloatFromCurve(FloatCurve=DataInterface() {propertyValues {Curve {keys [
         {arriveTangent=0.0466, interpMode=RCIM_Cubic, leaveTangent=0.0466, tangentMode=RCTM_User, time=0, value=0}
         {arriveTangent=0.0367, interpMode=RCIM_Cubic, leaveTangent=0.0367, tangentMode=RCTM_User, time=0.5, value=0}
         {arriveTangent=1.6703, interpMode=RCIM_Cubic, leaveTangent=1.6703, tangentMode=RCTM_User, time=1, value=1}
        ]}}})
    Index 0 Param 4=1
    Index 1 Param 1=&FloatFromCurve(FloatCurve=DataInterface() {propertyValues {Curve {keys [<at 0 1>,<at 0.5 1>,<at 1 0>]}}})
    Module.Param0WriteEnabled=true
    Module.Param0WriteEnabled1=true
    Module.Param1WriteEnabled=true
    Module.Param1WriteEnabled1=true
    Module.Param2WriteEnabled=true
    Module.Param2WriteEnabled1=true
    Module.Param3WriteEnabled=true
    Module.Param3WriteEnabled1=true
    Write Parameter Index 1=true
   )
   &SolveForcesAndVelocity(Clamp Velocity=false, Limit Acceleration=false)
   &ScaleColor(
    Scale Alpha=&Multiply_Float(
     A=&FNTimeOfDayScalar(Night Scale=0.5, Morning Scale=1, Evening Scale=1)
     B=&Multiply_Float(
      A=&FN_FrontEndFloatModifier(Gameplay=1, Front End=0.5)
      B=&FN_PickFloatDependOnPlatform(TargetPlatformActive=1, TargetPlatformNoActive=0.6)
     )
    )
    ScaleA=true
    ScaleRGB=true
   )
  ]
  rendererClasses [&NiagaraMeshRendererProperties]
  renderers [
   {rendererClass=&NiagaraMeshRendererProperties, rendererIndex=0}
   {rendererClass=&NiagaraMeshRendererProperties, rendererIndex=1}
  ]
 }
 Trail {
  emitterUpdateScript [
   &EmitterState(
    Inactive Response=::NewEnumerator1
    Life Cycle Mode=::NewEnumerator1
    Loop Behavior=::NewEnumerator1
    Loop Duration Mode=::NewEnumerator1
    UseLoopDelay=false
   )
   &SpawnBurst_Instantaneous(Spawn Count=2, Use Spawn Probability=false, UseLoopCountLimit=false)
  ]
  particleSpawnScript [
   &PartitionParticles()
   &InitializeParticle(
    Color=<rgba 1 1 1 2.5>
    Lifetime=5
    Mesh Scale=<1 1 0.8>
    Mesh Scale Mode=::NewEnumerator4
    UsePositionOffset=false
    Write Material Random=true
   )
  ]
  particleUpdateScript [
   &ParticleState(Kill Particles When Lifetime Has Elapsed=false, Let Infinitely Lived Particles Die When Emitter Deactivates=true)
   &ScaleMeshSize(Scale Factor=&MakeCustomVectorFromBool(Bool=$User.bIsFrontEndPreview, True Vector=<0.75 0.75 0.9>, False Vector=<1 1 1>))
   &SkeletalMeshLocation(
    Bone / Socket Index=&SelectIntFromArray(Direct Array Index=$Particles.PartitionParticles.Partition, Int Selection Array=DataInterface() {propertyValues {IntData [0,1]}})
    Bone Sampling Mode=::NewEnumerator4
    MaskSpawnGroup=false
    Orientation Sampling=::NewEnumerator1
    Sampled Position Offset=&SelectVectorFromArray(
     Direct Array Index=$Particles.PartitionParticles.Partition
     Vector Selection Array=DataInterface() {propertyValues {FloatData [<0 0 0>, <0 0 0>]}}
    )
    Skeletal Mesh=$System.Skeletal Mesh
   )
   &DynamicMaterialParameters(
    Index 0 Param 1=&RandomRangeFloat(Minimum=0.1, Maximum=0.5)
    Index 0 Param 2=&RandomRangeFloat(Minimum=0.1, Maximum=1)
    Index 0 Param 3=10
    Module.Index0Param0Write=true
    Module.Index0Param1Write=true
    Module.Index0Param2Write=true
    Module.Index0Param3Write=true
   )
   &ScaleColor(
    Scale Alpha=&Multiply_Float(A=&FN_FrontEndFloatModifier(Gameplay=1, Front End=0.8), B=1)
    Scale RGB=&Lerp_Vector(
     A=&Multiply_Vector(
      A=&MakeVectorFromLinearColorRGB(Color=<rgba 1 1 1 1>)
      B=&MakeVectorFromLinearColorRGB(Color=&FNTimeOfDayColor(Night Color=<rgba 0.4 0.55 1 1>, Morning Color=<rgba 1 1 1 1>, Day Color=<rgba 1 1 1 1>, Evening Color=<rgba 1 1 1 1>))
     )
     B=&MakeVectorFromLinearColorRGB(Color=<rgba 5 2.4 0.1543 1>)
     Alpha=$User.ParaGlideLeanAlpha
    )
    ScaleA=true
    ScaleRGB=true
   )
  ]
  rendererClasses [&NiagaraMeshRendererProperties]
  renderers [{rendererClass=&NiagaraMeshRendererProperties, rendererIndex=0}]
 }
 TrailLong {
  emitterUpdateScript [
   &EmitterState(
    Inactive Response=::NewEnumerator1
    Life Cycle Mode=::NewEnumerator1
    Loop Behavior=::NewEnumerator1
    Loop Duration Mode=::NewEnumerator1
    UseLoopDelay=false
   )
   &SpawnBurst_Instantaneous(Spawn Count=2, Use Spawn Probability=false, UseLoopCountLimit=false)
   /SetVariables(Emitter.Texture_Speed_Master:NiagaraFloat=&MakeCustomFloatFromBool(
     Bool=&SetBoolByFloatComparison(A=1, B=$User.ParaGlideLeanAlpha, Comparison Type=::NewEnumerator4)
     True Float=1.3
     False Float=&RandomRangeFloat(Minimum=0.4, Maximum=0.5)
    )) {isSetParameters=true}
   /SetVariables(Emitter.Texture_Speed_Master_Detail:NiagaraFloat=&MakeCustomFloatFromBool(
     Bool=&SetBoolByFloatComparison(A=1, B=$User.ParaGlideLeanAlpha, Comparison Type=::NewEnumerator4)
     True Float=0.6
     False Float=0.45
    )) {isSetParameters=true}
   /SetVariables(Emitter.SphereMaskU:NiagaraFloat=&Lerp_Float(A=1, B=0.95, Alpha=$User.ParaGlideLeanAlpha)) {isSetParameters=true}
  ]
  particleSpawnScript [
   &PartitionParticles()
   &InitializeParticle(
    Color Maximum=<rgba 0.02 0.02 0.02 1.5>
    Color Minimum=<rgba 0.01 0.01 0.01 1>
    Color Mode=::NewEnumerator2
    Mesh Scale=&SelectVectorFromArray(
     Direct Array Index=$Particles.PartitionParticles.Partition
     Vector Selection Array=DataInterface() {propertyValues {FloatData [<0.6 1 0.6>, <0.6 1 0.6>]}}
    )
    Mesh Scale Mode=::NewEnumerator4
    UsePositionOffset=true
    Write Material Random=true
   )
   &InitialMeshOrientation(Mesh Orientation Mode=::NewEnumerator2, Rotation=<0.25 0 0>, Use Rotation Vector=true)
   &InitialMeshOrientation(
    Mesh Orientation Mode=::NewEnumerator2
    Rotation=&SelectVectorFromArray(
     Direct Array Index=$Particles.PartitionParticles.Partition
     Vector Selection Array=DataInterface() {propertyValues {FloatData [<0 0 0>, <0 0 0.1>]}}
    )
    Use Rotation Vector=true
   ) {module=InitialMeshOrientation001}
   /SetVariables(Particles.MeshQuat_Yellow:Quat4f=/Niagara/DynamicInputs/Arrays/SelectQuaternionFromArray(
     Array Sampling Mode:ENiagaraArraySamplingMode=::NewEnumerator1
     Direct Array Index:NiagaraInt32=$Particles.PartitionParticles.Partition
     Interpolation:NiagaraFloat=?
     Direct Array Mode:ENiagara_WrapClamp=::NewEnumerator0
     Evaluation Type:ENiagaraRandomnessEvaluation=::NewEnumerator0
     Normalized Interpolation:NiagaraBool=true
     Recalculate Random Each Loop:NiagaraBool=false
     Randomness Mode:ENiagaraRandomnessMode=::NewEnumerator0
     Random Seed:NiagaraInt32=?
     Fixed Random Seed:NiagaraBool=false
     Quaternion Selection Array:NiagaraDataInterfaceArrayQuat=DataInterface() {propertyValues {GpuSyncMode=SyncCpuToGpu, MaxElements=0, QuatData [<0 0 0 1>, <0 0 0 1>]}}
     Enable Spherical Linear Interpolation:NiagaraBool=?
    )) {isSetParameters=true}
  ]
  particleUpdateScript [
   &ParticleState(Kill Particles When Lifetime Has Elapsed=false, Let Infinitely Lived Particles Die When Emitter Deactivates=true)
   &ScaleMeshSize(Scale Factor=&Lerp_Vector(A=<1 0.9 1>, B=<1 2 1>, Alpha=$User.ParaGlideLeanAlpha))
   &SkeletalMeshLocation(
    Bone / Socket Index=&SelectIntFromArray(Direct Array Index=$Particles.PartitionParticles.Partition, Int Selection Array=DataInterface() {propertyValues {IntData [0,1]}})
    Bone Sampling Mode=::NewEnumerator4
    MaskSpawnGroup=false
    Orientation Sampling=::NewEnumerator1
    Sampled Position Offset=&Lerp_Vector(
     A=&SelectVectorFromArray(
      Direct Array Index=$Particles.PartitionParticles.Partition
      Vector Selection Array=DataInterface() {propertyValues {FloatData [<0 0 0>, <0 0 0>]}}
     )
     B=&SelectVectorFromArray(
      Direct Array Index=$Particles.PartitionParticles.Partition
      Vector Selection Array=DataInterface() {propertyValues {FloatData [<0 0 -10>, <0 0 10>]}}
     )
     Alpha=$User.ParaGlideLeanAlpha
    )
    Skeletal Mesh=$System.Skeletal Mesh
   )
   &DynamicMaterialParameters(
    Index 0 Param 1=&RandomRangeFloat(Minimum=0.2, Maximum=1)
    Index 0 Param 2=&RandomRangeFloat(Minimum=0.2, Maximum=1)
    Index 0 Param 3=&FN_FrontEndFloatModifier(Gameplay=0.5, Front End=0)
    Index 0 Param 4=20
    Index 3 Param 1=150
    Index 3 Param 4=&FloatFromCurve(FloatCurve=DataInterface() {propertyValues {Curve {keys [<at 0 0>,<at 1 1>]}}})
    Module.Index0Param0Write=true
    Module.Index0Param1Write=true
    Module.Index0Param2Write=true
    Module.Index0Param3Write=true
    Module.Index2Param0Write=true
    Module.Index2Param1Write=true
    Module.Index2Param2Write=true
    Module.Index2Param3Write=true
    Module.Index3Param0Write=true
    Module.Index3Param1Write=true
    Module.Index3Param2Write=true
    Module.Index3Param3Write=true
    Write Parameter Index 2=true
    Write Parameter Index 3=true
   )
   &InterpolateOverTime(
    Rate Of Change:NiagaraFloat=&FN_FrontEndFloatModifier(Gameplay=10, Front End=10)
    Target Vector Value:Vector3f=&AddVectorToPosition(
     Position=$Particles.Position
     Vector=&TransformVector(
      Vector=&FN_FrontEndVector3Modifier(Gameplay=$System.IGCenterposition, Front End=$System.FECenterposition)
      Destination Space=::Simulation
     )
    )
    Variable Type=::NewEnumerator2
   )
   /SetVariables(Particles.Interpolate Dynamic=&MakeFloat4FromVectorAndFloat(
     "Vector (RGB)"=$StackContext.InterpolateOverTime.InterpolateOverTime_V3_.Moving Average
     "Float (Alpha)"=&FN_FrontEndFloatModifier(Gameplay=0, Front End=0)
    )) {isSetParameters=true}
   /SetVariables(Particles.Color002=&Multiply_LinearColorByFloat(
     Linear Color=&FNTimeOfDayColor(
      Night Color=<rgba 0.012 0.0165 0.03 0.8>
      Morning Color=<rgba 0.015 0.015 0.015 0.8>
      Day Color=<rgba 0.015 0.015 0.015 0.8>
      Evening Color=<rgba 0.015 0.015 0.015 0.8>
     )
     Float=&Multiply_Float(
      A=&FN_FrontEndFloatModifier(Gameplay=1, Front End=0.6)
      B=&FN_PickFloatDependOnPlatform(TargetPlatformActive=1, TargetPlatformNoActive=0.7)
     )
    )) {isSetParameters=true}
   /SetVariables(Particles.Size_Detail:Vector3f=&Lerp_Vector(A=<0.6 0.8 0.6>, B=<0.6 1.5 0.6>, Alpha=$User.ParaGlideLeanAlpha)) {isSetParameters=true}
   &ScaleColor(
    Scale Alpha=&FloatFromCurve(
     FloatCurve=DataInterface() {propertyValues {Curve {keys [<at 0 0>,<at 0.1 1>]}}}
     Scale Curve=&Multiply_Float(A=&FN_FrontEndFloatModifier(Gameplay=1, Front End=0.6), B=1)
    )
    Scale RGB=&Multiply_Vector(
     A=&MakeVectorFromLinearColorRGB(Color=&FNTimeOfDayColor(Night Color=<rgba 0.5 0.625 1 1>, Morning Color=<rgba 1 1 1 1>, Day Color=<rgba 1 1 1 1>, Evening Color=<rgba 1 1 1 1>))
     B=&FN_FrontEndVector3Modifier(Gameplay=<1 1 1>, Front End=<0.5 0.5 0.5>)
    )
    ScaleA=true
    ScaleRGB=true
   )
  ]
  rendererClasses [&NiagaraMeshRendererProperties]
  renderers [
   {rendererClass=&NiagaraMeshRendererProperties, rendererIndex=0}
   {rendererClass=&NiagaraMeshRendererProperties, rendererIndex=1}
  ]
 }
 Trailspiral {
  emitterUpdateScript [
   &EmitterState(
    Inactive Response=::NewEnumerator1
    Life Cycle Mode=::NewEnumerator1
    Loop Behavior=::NewEnumerator1
    Loop Duration Mode=::NewEnumerator1
    UseLoopDelay=false
   )
   &SpawnRate(
    SpawnRate=&MakeCustomFloatFromBool(
     Bool=&SetBoolByFloatComparison(A=1, B=$User.ParaGlideLeanAlpha, Comparison Type=::NewEnumerator4)
     True Float=15
     False Float=10
    )
    Use Spawn Probability=false
   )
  ]
  particleSpawnScript [
   &PartitionParticles()
   &InitializeParticle(
    Color=&MakeCustomLinearColorFromBool(
     Bool=&SetBoolByFloatComparison(A=1, B=$User.ParaGlideLeanAlpha, Comparison Type=::NewEnumerator4)
     True Color=&RandomRangeLinearColor(Minimum=<rgba 0.5 0.1995 0.0172 2>, Maximum=<rgba 0.8 0.4267 0.057 1>)
     False Color=&RandomRangeLinearColor(Minimum=<rgba 0.05 0.05 0.05 1.6>, Maximum=<rgba 0.3 0.3 0.3 2>)
    )
    Lifetime=&MakeCustomFloatFromBool(
     Bool=&SetBoolByFloatComparison(A=1, B=$User.ParaGlideLeanAlpha, Comparison Type=::NewEnumerator4)
     True Float=0.7
     False Float=1.25
    )
    Mesh Scale Max=<2 1.5 1.5>
    Mesh Scale Min=<1.2 1.1 1.1>
    Mesh Scale Mode=::NewEnumerator2
    UsePositionOffset=false
    Write Material Random=true
   )
   &InitialMeshOrientation(Mesh Orientation Mode=::NewEnumerator2, Rotation=<0 0.25 0>, Use Rotation Vector=true)
   &InitialMeshOrientation(
    Mesh Orientation Mode=::NewEnumerator2
    Rotation=&RandomRangeVector(Minimum=<0 0 -1>, Maximum=<0 0 1>)
    Use Rotation Vector=true
   ) {module=InitialMeshOrientation001}
   &ShapeLocation(
    Cylinder Height=30
    Cylinder Radius=10
    MaskSpawnGroup=false
    Override Seed=false
    Rotation Mode=::NewEnumerator1
    Shape Primitive=::NewEnumerator1
    Surface Only=false
    UseOwnerScale=true
   )
   &AddVelocity(Velocity=&MakeCustomVectorFromBool(
     Bool=&SetBoolByFloatComparison(A=1, B=$User.ParaGlideLeanAlpha, Comparison Type=::NewEnumerator4)
     True Vector=<0 0 80>
     False Vector=<1 0 30>
    ))
  ]
  particleUpdateScript [
   &ParticleState()
   &ScaleMeshSize(Scale Factor=&Lerp_Vector(A=<1 1 1>, B=<1.3 1 1>, Alpha=$User.ParaGlideLeanAlpha))
   &SkeletalMeshLocation(
    Bone / Socket Index=&SelectIntFromArray(Direct Array Index=$Particles.PartitionParticles.Partition, Int Selection Array=DataInterface() {propertyValues {IntData [0,1]}})
    Bone Sampling Mode=::NewEnumerator4
    MaskSpawnGroup=false
    Orientation Sampling=::NewEnumerator1
    Position Sampling=::NewEnumerator1
    Sampled Position Offset=&AddVectorToPosition(
     Position=Linked() {linkedVariable {name=Particles.Initial.ShapeLocation.ShapePosition, type=Vector3f}}
     Vector=&SelectVectorFromArray(
      Direct Array Index=$Particles.PartitionParticles.Partition
      Vector Selection Array=DataInterface() {propertyValues {FloatData [<0 0 0>, <0 0 0>]}}
     )
    )
    Skeletal Mesh=$System.Skeletal Mesh
   )
   &DynamicMaterialParameters(
    Index 0 Param 3=1
    Index 0 Param 4=&FloatFromCurve(FloatCurve=DataInterface() {propertyValues {Curve {keys [<at 0 0>,<at 1 1>]}}})
    Index 3 Param 1=500
    Module.Index0Param0Write=true
    Module.Index0Param1Write=true
    Module.Index0Param2Write=true
    Module.Index0Param3Write=true
    Module.Index3Param0Write=true
    Module.Index3Param1Write=true
    Module.Index3Param2Write=true
    Module.Index3Param3Write=true
    Write Parameter Index 3=true
   )
   &ScaleColor(
    Scale Alpha=&FloatFromCurve(
     FloatCurve=DataInterface() {propertyValues {Curve {keys [
         {interpMode=RCIM_Cubic, tangentMode=RCTM_User, time=0, value=0}
         {interpMode=RCIM_Cubic, tangentMode=RCTM_User, time=0.4, value=0.8}
         {tangentMode=RCTM_User, time=1, value=0}
        ]}}}
     Scale Curve=&Multiply_Float(A=0.8, B=&FN_FrontEndFloatModifier(Gameplay=1, Front End=0.8))
    )
    Scale RGB=&MakeCustomVectorFromBool(
     Bool=$User.bIsFrontEndPreview
     True Vector=&Multiply_VectorByFloat(
      Vector=&MakeVectorFromLinearColorRGB(Color=&ColorFromCurve(ColorCurve=DataInterface() {propertyValues {
          AlphaCurve {keys []}
          BlueCurve {keys [<at 0.4142 0.005 RCIM_Cubic>,<at 1 1 RCIM_Cubic>]}
          GreenCurve {keys [<at 0.4142 0.0036 RCIM_Cubic>,<at 1 0.8187 RCIM_Cubic>]}
          RedCurve {keys [<at 0.4142 0.0024 RCIM_Cubic>,<at 1 0.6771 RCIM_Cubic>]}
         }}))
      Float=&RandomRangeFloat(Minimum=0.5, Maximum=1)
     )
     False Vector=&Multiply_VectorByFloat(
      Vector=&Multiply_Vector(
       A=&MakeVectorFromLinearColorRGB(Color=&ColorFromCurve(ColorCurve=DataInterface() {propertyValues {
           AlphaCurve {keys []}
           BlueCurve {keys [<at 0.2735 0.005 RCIM_Cubic>,<at 1 1 RCIM_Cubic>]}
           GreenCurve {keys [<at 0.2735 0.005 RCIM_Cubic>,<at 1 1 RCIM_Cubic>]}
           RedCurve {keys [<at 0.2735 0.005 RCIM_Cubic>,<at 1 1 RCIM_Cubic>]}
          }}))
       B=&MakeVectorFromLinearColorRGB(Color=&FNTimeOfDayColor(Night Color=<rgba 0.49 0.5425 0.7 1>, Morning Color=<rgba 1 1 1 1>, Day Color=<rgba 1 1 1 1>, Evening Color=<rgba 1 1 1 1>))
      )
      Float=&RandomRangeFloat(Minimum=0.5, Maximum=1)
     )
    )
    ScaleA=true
    ScaleRGB=true
   )
   &SolveForcesAndVelocity(Clamp Velocity=false, Limit Acceleration=false)
  ]
  rendererClasses [&NiagaraMeshRendererProperties]
  renderers [{rendererClass=&NiagaraMeshRendererProperties, rendererIndex=0}]
 }
 mask {
  emitterUpdateScript [
   &EmitterState(
    Inactive Response=::NewEnumerator1
    Life Cycle Mode=::NewEnumerator1
    Loop Behavior=::NewEnumerator1
    Loop Duration Mode=::NewEnumerator1
    UseLoopDelay=false
   )
   &SpawnBurst_Instantaneous(Spawn Count=4, Use Spawn Probability=false, UseLoopCountLimit=false)
  ]
  particleSpawnScript [
   &PartitionParticles(Number Of Partitions=4)
   &InitializeParticle(
    Color=<rgba 1 1 1 2>
    Lifetime=5
    Mesh Scale Mode=::NewEnumerator3
    Mesh Uniform Scale=&SelectFloatFromArray(
     Direct Array Index=$Particles.PartitionParticles.Partition
     Float Selection Array=DataInterface() {propertyValues {FloatData [1.3,1.3,1.2,1.2]}}
    )
    UsePositionOffset=false
    Write Material Random=true
   )
   /SetVariables(Particles.MeshIndex2:NiagaraInt32=&SelectIntFromArray(Direct Array Index=$Particles.PartitionParticles.Partition, Int Selection Array=DataInterface() {propertyValues {IntData [0,0,1,1]}})) {isSetParameters=true}
   &InitialMeshOrientation(
    Mesh Orientation Mode=::NewEnumerator2
    Rotation=&SelectVectorFromArray(
     Direct Array Index=$Particles.PartitionParticles.Partition
     Vector Selection Array=DataInterface() {propertyValues {FloatData [<0 0 0.5586>, <0 0 0.4414>, <0 0 0.3964>, <0 0 0.6126>]}}
    )
    Use Rotation Vector=true
   )
   &InitialMeshOrientation(
    Mesh Orientation Mode=::NewEnumerator2
    Rotation=&SelectVectorFromArray(
     Direct Array Index=$Particles.PartitionParticles.Partition
     Vector Selection Array=DataInterface() {propertyValues {FloatData [<-0.05 0 0>, <-0.05 0 0>, <-0.05 0 0>, <-0.05 0 0>]}}
    )
    Use Rotation Vector=true
   ) {module=InitialMeshOrientation001}
   &InitialMeshOrientation(
    Mesh Orientation Mode=::NewEnumerator2
    Rotation=&MakeCustomVectorFromBool(
     Bool=$User.bIsFrontEndPreview
     True Vector=&SelectVectorFromArray(
      Direct Array Index=$Particles.PartitionParticles.Partition
      Vector Selection Array=DataInterface() {propertyValues {FloatData [<0 0 0.4>, <0 0 0.6667>, <0 0 0.7568>, <0 0 0.3419>]}}
     )
     False Vector=<0 0 0>
    )
    Use Rotation Vector=true
   ) {module=InitialMeshOrientation002}
  ]
  particleUpdateScript [
   &ParticleState(Kill Particles When Lifetime Has Elapsed=false, Let Infinitely Lived Particles Die When Emitter Deactivates=true)
   &UpdateMeshOrientation(
    Rotation Rate=&Sine(
     Normalized Angle=$Emitter.Age
     Period=&MakeCustomFloatFromBool(
      Bool=&SetBoolByFloatComparison(A=1, B=$User.ParaGlideLeanAlpha, Comparison Type=::NewEnumerator4)
      True Float=0.4
      False Float=2.2
     )
     Scale=&MakeCustomFloatFromBool(
      Bool=&SetBoolByFloatComparison(A=1, B=$User.ParaGlideLeanAlpha, Comparison Type=::NewEnumerator4)
      True Float=1.1
      False Float=0.4
     )
     Bias=0
    )
    Rotation Vector=&SelectVectorFromArray(
     Direct Array Index=$Particles.PartitionParticles.Partition
     Vector Selection Array=DataInterface() {propertyValues {FloatData [<0 0 0.3>, <0.3 0 0.02>, <-0.2 0.05 0>, <0.05 0 0.3>]}}
    )
   )
   &SkeletalMeshLocation(
    Bone / Socket Index=&SelectIntFromArray(Direct Array Index=$Particles.PartitionParticles.Partition, Int Selection Array=DataInterface() {propertyValues {IntData [0,1,0,1]}})
    Bone Sampling Mode=::NewEnumerator4
    MaskSpawnGroup=false
    Orientation Sampling=::NewEnumerator1
    Sampled Position Offset=&Multiply_Vector(
     A=&SelectVectorFromArray(
      Direct Array Index=$Particles.PartitionParticles.Partition
      Vector Selection Array=DataInterface() {propertyValues {FloatData [<20 0 60>, <-20 0 60>, <-13 0 40>, <13 0 40>]}}
     )
     B=&MakeVector(
      X=&Sine(
       Normalized Angle=$Emitter.Age
       Period=&RandomRangeFloat(Minimum=1.5, Maximum=2)
       Scale=&RandomRangeFloat(Minimum=0.15, Maximum=0.2)
       Bias=1
      )
      Y=&Sine(
       Normalized Angle=$Emitter.Age
       Period=&RandomRangeFloat(Minimum=1.5, Maximum=2)
       Scale=&RandomRangeFloat(Minimum=0.15, Maximum=0.2)
       Bias=1
      )
      Z=&Sine(
       Normalized Angle=$Emitter.Age
       Period=&RandomRangeFloat(Minimum=1.5, Maximum=2)
       Scale=&RandomRangeFloat(Minimum=0.2, Maximum=0.3)
       Bias=1
      )
     )
    )
    Sampled Position Offset Coordinate Space:ENiagaraSkelMeshTransforms=::NewEnumerator1
    Skeletal Mesh=$System.Skeletal Mesh
   )
  ]
  rendererClasses [&NiagaraMeshRendererProperties]
  renderers [{rendererClass=&NiagaraMeshRendererProperties, rendererIndex=0}]
 }
 smoke {
  emitterUpdateScript [
   &EmitterState(Life Cycle Mode=::NewEnumerator1, Loop Duration=1, UseLoopDelay=false)
   &SpawnRate(
    SpawnRate=&MakeCustomFloatFromBool(
     Bool=&SetBoolByFloatComparison(A=1, B=$User.ParaGlideLeanAlpha, Comparison Type=::NewEnumerator4)
     True Float=10
     False Float=8
    )
    Use Spawn Probability=false
   )
  ]
  particleSpawnScript [
   &PartitionParticles()
   &InitializeParticle(
    Color Maximum=<rgba 0.15 0.15 0.15 1>
    Color Minimum=<rgba 0.05 0.05 0.05 1>
    Color Mode=::NewEnumerator2
    Lifetime Min=1.5
    Lifetime Mode=::NewEnumerator1
    Sprite Rotation Mode=::NewEnumerator2
    Sprite Size Mode=::NewEnumerator1
    Uniform Sprite Size Max=50
    Uniform Sprite Size Min=30
    UsePositionOffset=false
    Write Material Random=true
   )
   &AddVelocity(Velocity=&MakeCustomVectorFromBool(
     Bool=&SetBoolByFloatComparison(A=1, B=$User.ParaGlideLeanAlpha, Comparison Type=::NewEnumerator4)
     True Vector=<0 0 160>
     False Vector=<0 0 100>
    ))
  ]
  particleUpdateScript [
   &ParticleState()
   &ScaleSpriteSize(
    Uniform Curve Scale=&FN_FrontEndFloatModifier(Gameplay=1, Front End=1.2)
    Uniform Curve Sprite Scale=DataInterface() {propertyValues {Curve {keys [
        {arriveTangent=2.258, interpMode=RCIM_Cubic, leaveTangent=2.258, tangentMode=RCTM_User, time=0, value=0.5}
        {arriveTangent=0.5, interpMode=RCIM_Cubic, leaveTangent=0.5, time=0.2, value=0.85}
        <at 1 1 RCIM_Cubic>
       ]}}}
   )
   &SubUVAnimation(
    "Sprite Renderer "=DataInterface() {propertyValues {SpriteRenderer=&self:smoke_0.NiagaraSpriteRendererProperties_1}}
    UseEndFrame=true
    UseStartFrame=true
   )
   &SkeletalMeshLocation(
    Bone / Socket Index=&SelectIntFromArray(Direct Array Index=$Particles.PartitionParticles.Partition, Int Selection Array=DataInterface() {propertyValues {IntData [0,1]}})
    Bone Sampling Mode=::NewEnumerator4
    MaskSpawnGroup=false
    Orientation Sampling=::NewEnumerator1
    Position Sampling=::NewEnumerator1
    Skeletal Mesh=$System.Skeletal Mesh
   )
   &ScaleColor(
    Scale Alpha=&FloatFromCurve(
     FloatCurve=DataInterface() {propertyValues {Curve {keys [
         {arriveTangent=7.2065, interpMode=RCIM_Cubic, leaveTangent=7.2065, tangentMode=RCTM_User, time=0, value=0}
         {arriveTangent=0.0408, interpMode=RCIM_Cubic, leaveTangent=0.0408, tangentMode=RCTM_User, time=0.2, value=1}
         {arriveTangent=-1.25, interpMode=RCIM_Cubic, leaveTangent=-1.25, time=0.6, value=0.8}
         <at 1 0 RCIM_Cubic>
        ]}}}
     Scale Curve=&Multiply_Float(
      A=&RandomRangeFloat(Minimum=0.6, Maximum=1)
      B=&FN_FrontEndFloatModifier(Gameplay=1, Front End=0.6)
     )
    )
    Scale RGB=&MakeVectorFromLinearColorRGB(Color=&FNTimeOfDayColor(Night Color=<rgba 0.3 0.4928 0.8 1>, Morning Color=<rgba 1 1 1 1>, Day Color=<rgba 1 1 1 1>, Evening Color=<rgba 1 1 1 1>))
    ScaleA=true
    ScaleRGB=true
   )
   &Drag(
    Drag=&FN_FrontEndFloatModifier(Gameplay=0, Front End=0.5)
    Use Linear Drag=true
    Use Rotational Drag=true
   )
   &SolveForcesAndVelocity(Clamp Velocity=false, Limit Acceleration=false)
  ]
  rendererClasses [&NiagaraSpriteRendererProperties]
  renderers [{rendererClass=&NiagaraSpriteRendererProperties}]
 }
}
summary {
 systemName=NS_Contrail_MistMylar
 emitters [
  {emitterName=Trail, rendererClasses [&NiagaraMeshRendererProperties]}
  {emitterName=mask, rendererClasses [&NiagaraMeshRendererProperties]}
  {emitterName=TrailLong, rendererClasses [&NiagaraMeshRendererProperties]}
  {emitterName=smoke, rendererClasses [&NiagaraSpriteRendererProperties]}
  {emitterName=Lightning, rendererClasses [&NiagaraMeshRendererProperties]}
  {emitterName=Trailspiral, rendererClasses [&NiagaraMeshRendererProperties]}
 ]
 userVariables [
  User.bIsFrontEnd {type=NiagaraBool}
  User.bIsFrontEndPreview {type=NiagaraBool}
  User.ParaGlideLeanAlpha {type=NiagaraFloat, defaultValue=0, description=""}
 ]
}
